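# Looping foley Sound Cue for the DenimEquip emote, likely its forward-movement loop, using emote attenuation and the emote foley sound class.
# A mixer layers looping Shimmer and Wind waves, each through a continuous modulator driven by the EmoteMoveSpeed parameter.
# The parameter scales each layer's volume over its own range, with a non-zero floor on both.
# It plays when silent, overrides concurrency with a small voice cap, and has an audio-analysis submix send.
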SoundCue /BRCosmetics/Sounds/Emotes/DenimEquip/EmoteFoley_DenimEquip_FwdLP_Cue
root=Mixer_0
nodes [
 {id=Mixer_0, children [ModulatorContinuous_1, ModulatorContinuous_0], properties {inputVolume [1,1]}}
 {
  id=ModulatorContinuous_0
  children [WavePlayer_3]
  properties {pitchModulationParams {minOutput=1, parameterName=EmoteMoveSpeed}, volumeModulationParams {minOutput=0.3, parameterName=EmoteMoveSpeed}}
 }
 {
  id=ModulatorContinuous_1
  children [WavePlayer_2]
  properties {
   pitchModulationParams {minOutput=1, parameterName=EmoteMoveSpeed}
   volumeModulationParams {maxOutput=2, minOutput=1, parameterName=EmoteMoveSpeed}
  }
 }
 {id=WavePlayer_2, properties {bLooping=true, soundWaveAssetPtr=&/Emote_DenimEquip_Shimmer_LP}}
 {id=WavePlayer_3, properties {bLooping=true, soundWaveAssetPtr=&/Emote_DenimEquip_Wind_LP}}
]
properties {
 attenuationSettings=&Fort_Emotes_Athena_Att
 audioPropertiesSheet=/BRCosmetics/Sounds/Settings/PropertySheets/Emotes/APS_Cosmetic_Emote_Foley
 bOverrideConcurrency=true
 maxDistance=1600
 soundClassObject=/Game/Sounds/FortSoundClassesAndModes/Fort_Foley_Emote
 virtualizationMode=PlayWhenSilent
 volumeMultiplier=1.2
 concurrencyOverrides {maxCount=3}
 soundSubmixSends [{minSendLevel=0, soundSubmix=&Fortnite_AudioAnalysis_Music}]
}
