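# Looping sound cue for the DenimEquip MiniGuff emote foley, with attenuation, emote foley sound class and limited concurrency.
# A mixer layers a looping flapping bed with vocal chatter that repeats at random intervals.
# Three chatter loops pick from small subsets of clips at raised, staggered pitch and timing, suggesting several small voices.
# A quieter fourth loop draws on all eight clips at near-original pitch with longer gaps.
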SoundCue /BRCosmetics/Sounds/Emotes/DenimEquip/EmoteFoley_DenimEquip_MiniGuff_LP_Cue
root=Mixer_0
nodes [
 {id=Delay_0, children [Random_3], properties {delayMax=1.5, delayMin=0.5}}
 {id=Delay_1, children [Random_2], properties {delayMax=2, delayMin=1}}
 {id=Delay_2, children [Random_0], properties {delayMax=1}}
 {id=Delay_3, children [Random_4], properties {delayMax=3, delayMin=2}}
 {id=Looping_1, children [Modulator_0]}
 {id=Looping_2, children [Modulator_1]}
 {id=Looping_3, children [Modulator_2]}
 {id=Looping_4, children [Modulator_3]}
 {id=Mixer_0, children [WavePlayer_0,Mixer_2,Looping_4], properties {inputVolume [1,0.528,0.336]}}
 {id=Mixer_2, children [Looping_1,Looping_2,Looping_3], properties {inputVolume [0.9,0.75,1]}}
 {id=Modulator_0, children [Delay_0], properties {pitchMax=1.2, pitchMin=1.1, volumeMin=0.98}}
 {id=Modulator_1, children [Delay_1], properties {pitchMax=1.25, pitchMin=1.1, volumeMin=0.98}}
 {id=Modulator_2, children [Delay_2], properties {pitchMax=1.15, pitchMin=1.05, volumeMin=0.98}}
 {id=Modulator_3, children [Delay_3], properties {pitchMax=1, pitchMin=0.98, volumeMin=0.98}}
 {id=Random_0, children [WavePlayer_8, WavePlayer_9, WavePlayer_10], properties {weights [1,1,1]}}
 {id=Random_2, children [WavePlayer_5,WavePlayer_6,WavePlayer_7], properties {weights [1,1,1]}}
 {id=Random_3, children [WavePlayer_1,WavePlayer_4,WavePlayer_5], properties {weights [1,1,1]}}
 {
  id=Random_4
  children [WavePlayer_2, WavePlayer_3, WavePlayer_11, WavePlayer_12, WavePlayer_13, WavePlayer_14, WavePlayer_15, WavePlayer_16]
  properties {weights [1,1,1,1,1,1,1,1]}
 }
 {id=WavePlayer_0, properties {bLooping=true, soundWaveAssetPtr=&/Emote_DenimEquip_Flapping_LP}}
 {id=WavePlayer_1, properties {soundWaveAssetPtr=&/Emote_Vocal_DenimEquip_Chatter_08}}
 {id=WavePlayer_10, properties {soundWaveAssetPtr=&/Emote_Vocal_DenimEquip_Chatter_01}}
 {id=WavePlayer_11, properties {soundWaveAssetPtr=&/Emote_Vocal_DenimEquip_Chatter_06}}
 {id=WavePlayer_12, properties {soundWaveAssetPtr=&/Emote_Vocal_DenimEquip_Chatter_05}}
 {id=WavePlayer_13, properties {soundWaveAssetPtr=&/Emote_Vocal_DenimEquip_Chatter_04}}
 {id=WavePlayer_14, properties {soundWaveAssetPtr=&/Emote_Vocal_DenimEquip_Chatter_03}}
 {id=WavePlayer_15, properties {soundWaveAssetPtr=&/Emote_Vocal_DenimEquip_Chatter_02}}
 {id=WavePlayer_16, properties {soundWaveAssetPtr=&Emote_Vocal_DenimEquip_Chatter_01}}
 {id=WavePlayer_2, properties {soundWaveAssetPtr=&Emote_Vocal_DenimEquip_Chatter_08}}
 {id=WavePlayer_3, properties {soundWaveAssetPtr=&/Emote_Vocal_DenimEquip_Chatter_07}}
 {id=WavePlayer_4, properties {soundWaveAssetPtr=&Emote_Vocal_DenimEquip_Chatter_07}}
 {id=WavePlayer_5, properties {soundWaveAssetPtr=&Emote_Vocal_DenimEquip_Chatter_06}}
 {id=WavePlayer_6, properties {soundWaveAssetPtr=&Emote_Vocal_DenimEquip_Chatter_05}}
 {id=WavePlayer_7, properties {soundWaveAssetPtr=&Emote_Vocal_DenimEquip_Chatter_04}}
 {id=WavePlayer_8, properties {soundWaveAssetPtr=&Emote_Vocal_DenimEquip_Chatter_03}}
 {id=WavePlayer_9, properties {soundWaveAssetPtr=&Emote_Vocal_DenimEquip_Chatter_02}}
]
properties {
 attenuationSettings=&Fort_Emotes_Athena_Att
 audioPropertiesSheet=/BRCosmetics/Sounds/Settings/PropertySheets/Emotes/APS_Cosmetic_Emote_Foley
 bOverrideConcurrency=true
 maxDistance=1600
 soundClassObject=/Game/Sounds/FortSoundClassesAndModes/Fort_Foley_Emote
 virtualizationMode=PlayWhenSilent
 volumeMultiplier=1.2
 concurrencyOverrides {maxCount=3}
 soundSubmixSends [{minSendLevel=0, soundSubmix=&Fortnite_AudioAnalysis_Music}]
}
